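# Physics asset for the Palespooky back bling cosmetic, giving its small character figure jiggle motion on head, hair, arms and legs.
# Capsule bodies cover the dyn_ joints and chain off root, with the torso swing and twist locked to root and the hip allowed slight movement.
# Arms and head hang from the torso, legs from the hip, and eight hair chains from the head.
# Tight angular limits keep motion subtle, and hair twist is locked.
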PhysicsAsset /BRCosmetics/Accessories/FORT_Backpacks/Backpack_M_MED_Palespooky/Meshes/M_MED_Palespooky_Pack_Physics
bodies {
 dyn_L_arm_elbow_jnt [{shapeType=Capsule, length=3.25, radius=1.25, center=<3.1875 0 0>, rotation=<rot 90 0 0>}]
 dyn_L_arm_shoulder_jnt [{shapeType=Capsule, length=1, radius=1.25, center=<1.9968 0 0>, rotation=<rot 90 0 0>}]
 dyn_L_leg_hip_jnt [{shapeType=Capsule, length=1, radius=1.5, rotation=<rot 90 0 0>}]
 dyn_L_leg_knee_jnt [{shapeType=Capsule, length=4.5, radius=1.5, center=<3.4695 0.9546 0>, rotation=<rot 90 0 -9.2608>}]
 dyn_R_arm_elbow_jnt [{shapeType=Capsule, length=3.25, radius=1.25, center=<3.1875 0 0>, rotation=<rot 90 0 0>}]
 dyn_R_arm_shoulder_jnt [{shapeType=Capsule, length=1, radius=1.25, center=<1.9968 0 0>, rotation=<rot 90 0 0>}]
 dyn_R_leg_hip_jnt [{shapeType=Capsule, length=1, radius=1.5, rotation=<rot 90 0 0>}]
 dyn_R_leg_knee_jnt [{shapeType=Capsule, length=4.5, radius=1.5, center=<3.4695 0.9546 0>, rotation=<rot 90 0 -9.2608>}]
 dyn_hair_chain1_jnt_1 [{shapeType=Capsule, length=4, radius=0.7, center=<0.2236 0 0>, rotation=<rot 90 0 0>}]
 dyn_hair_chain1_jnt_2 [{shapeType=Capsule, length=5, radius=0.8, center=<0.3492 0.0909 0>, rotation=<rot 90 0 -14.5956>}]
 dyn_hair_chain2_jnt_1 [{shapeType=Capsule, length=3, radius=0.8, center=<1.8784 0 0>, rotation=<rot 90 0 0>}]
 dyn_hair_chain2_jnt_2 [{shapeType=Capsule, length=4, radius=0.6, center=<1.5394 -0.0001 0.653>, rotation=<rot 66.6174 180 180>}]
 dyn_hair_chain3_jnt_1 [{shapeType=Capsule, length=3, radius=0.6, center=<2.2035 0 0>, rotation=<rot 90 0 0>}]
 dyn_hair_chain4_jnt_1 [{shapeType=Capsule, length=2, radius=0.6, center=<0.4491 0.1326 0>, rotation=<rot 90 180 163.5501>}]
 dyn_hair_chain5_jnt_1 [{shapeType=Capsule, length=5, radius=1, center=<0.5763 0 0.0031>, rotation=<rot 85.4772 180 180>}]
 dyn_hair_chain5_jnt_2 [{shapeType=Capsule, length=3, radius=0.7, center=<0.1691 0.4778 -0.0364>, rotation=<rot 1.4502 -19.3861 94.114>}]
 dyn_hair_chain6_jnt_1 [{shapeType=Capsule, length=4, radius=1, center=<0.1701 0.0191 0.0343>, rotation=<rot 78.845 179.9994 173.7303>}]
 dyn_hair_chain7_jnt_2 [{shapeType=Capsule, length=4, radius=0.7, center=<1.7768 0 0>, rotation=<rot 90 0 0>}]
 dyn_hair_chain8_jnt_2 [{shapeType=Capsule, length=8, radius=0.6, center=<1.2951 0 -0.7864>, rotation=<rot 61.5556 0 0>}]
 dyn_head_jnt [{shapeType=Capsule, length=0.25, radius=4.25, center=<-2.9113 0 0>}]
 dyn_hip_down_jnt [{shapeType=Capsule, length=0, radius=1, center=<0 0 -0.8083>}]
 dyn_torso_jnt [{shapeType=Capsule, length=1, radius=3.5, center=<-1.1175 0 0>, rotation=<rot 90 0 0>}]
 root [{shapeType=Capsule, length=1, radius=1}]
}
constraints [
 {bone1Name=dyn_hip_down_jnt, bone2Name=root, swing1LimitDegrees=5, swing2LimitDegrees=5, twistLimitDegrees=20}
 {bone1Name=dyn_L_leg_hip_jnt, bone2Name=dyn_hip_down_jnt, swing1LimitDegrees=25, swing2LimitDegrees=10, twistLimitDegrees=5}
 {bone1Name=dyn_L_leg_knee_jnt, bone2Name=dyn_L_leg_hip_jnt, swing1LimitDegrees=15, swing2LimitDegrees=10, twistLimitDegrees=5}
 {bone1Name=dyn_R_leg_hip_jnt, bone2Name=dyn_hip_down_jnt, swing1LimitDegrees=25, swing2LimitDegrees=10, twistLimitDegrees=5}
 {bone1Name=dyn_R_leg_knee_jnt, bone2Name=dyn_R_leg_hip_jnt, swing1LimitDegrees=15, swing2LimitDegrees=10, twistLimitDegrees=5}
 {
  bone1Name=dyn_torso_jnt
  bone2Name=root
  swing1LimitDegrees=1
  swing1Motion=Locked
  swing2LimitDegrees=1
  swing2Motion=Locked
  twistLimitDegrees=1
  twistMotion=Locked
 }
 {bone1Name=dyn_head_jnt, bone2Name=dyn_torso_jnt, swing1LimitDegrees=7, swing2LimitDegrees=7, twistLimitDegrees=7}
 {bone1Name=dyn_hair_chain1_jnt_1, bone2Name=dyn_head_jnt, swing1LimitDegrees=6, swing2LimitDegrees=6, twistLimitDegrees=1, twistMotion=Locked}
 {
  bone1Name=dyn_hair_chain1_jnt_2
  bone2Name=dyn_hair_chain1_jnt_1
  swing1LimitDegrees=6
  swing2LimitDegrees=6
  twistLimitDegrees=1
  twistMotion=Locked
 }
 {bone1Name=dyn_hair_chain2_jnt_1, bone2Name=dyn_head_jnt, swing1LimitDegrees=3, swing2LimitDegrees=3, twistLimitDegrees=1, twistMotion=Locked}
 {
  bone1Name=dyn_hair_chain2_jnt_2
  bone2Name=dyn_hair_chain2_jnt_1
  swing1LimitDegrees=5
  swing2LimitDegrees=5
  twistLimitDegrees=1
  twistMotion=Locked
 }
 {bone1Name=dyn_hair_chain3_jnt_1, bone2Name=dyn_head_jnt, swing1LimitDegrees=3, swing2LimitDegrees=3, twistLimitDegrees=1, twistMotion=Locked}
 {bone1Name=dyn_hair_chain4_jnt_1, bone2Name=dyn_head_jnt, swing1LimitDegrees=3, swing2LimitDegrees=3, twistLimitDegrees=1, twistMotion=Locked}
 {
  bone1Name=dyn_hair_chain5_jnt_2
  bone2Name=dyn_hair_chain5_jnt_1
  swing1LimitDegrees=8
  swing2LimitDegrees=8
  twistLimitDegrees=1
  twistMotion=Locked
 }
 {bone1Name=dyn_hair_chain7_jnt_2, bone2Name=dyn_head_jnt, swing1LimitDegrees=8, swing2LimitDegrees=8, twistLimitDegrees=1, twistMotion=Locked}
 {bone1Name=dyn_hair_chain8_jnt_2, bone2Name=dyn_head_jnt, swing1LimitDegrees=8, swing2LimitDegrees=8, twistLimitDegrees=1, twistMotion=Locked}
 {bone1Name=dyn_L_arm_shoulder_jnt, bone2Name=dyn_torso_jnt, swing1LimitDegrees=5, swing2LimitDegrees=8, twistLimitDegrees=3}
 {bone1Name=dyn_L_arm_elbow_jnt, bone2Name=dyn_L_arm_shoulder_jnt, swing1LimitDegrees=10, swing2LimitDegrees=20, twistLimitDegrees=5}
 {bone1Name=dyn_R_arm_shoulder_jnt, bone2Name=dyn_torso_jnt, swing1LimitDegrees=5, swing2LimitDegrees=8, twistLimitDegrees=3}
 {bone1Name=dyn_R_arm_elbow_jnt, bone2Name=dyn_R_arm_shoulder_jnt, swing1LimitDegrees=10, swing2LimitDegrees=20, twistLimitDegrees=5}
 {
  bone1Name=dyn_hair_chain5_jnt_1
  bone2Name=dyn_head_jnt
  swing1LimitDegrees=5
  swing2LimitDegrees=5
  twistLimitDegrees=45
  twistMotion=Locked
 }
 {
  bone1Name=dyn_hair_chain6_jnt_1
  bone2Name=dyn_head_jnt
  swing1LimitDegrees=4
  swing2LimitDegrees=7
  twistLimitDegrees=45
  twistMotion=Locked
 }
]
