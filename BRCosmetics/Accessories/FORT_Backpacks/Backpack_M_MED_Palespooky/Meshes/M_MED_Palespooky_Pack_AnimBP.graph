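# Animation blueprint for the Palespooky backpack cosmetic, adding secondary physics motion to its skeleton.
# The input pose feeds a RigidBody node using the pack's own physics asset, simulated in root bone space with damped component motion and overridden gravity.
# The simulation's alpha follows the current LOD, running at full strength at LOD 0 and mostly faded at LOD 1 and above.
# The event graph and rotation and translation variables are unused.
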AnimBlueprint /BRCosmetics/Accessories/FORT_Backpacks/Backpack_M_MED_Palespooky/Meshes/M_MED_Palespooky_Pack_AnimBP
target_skeleton=&/M_MED_Palespooky_Pack_Skeleton
graphs [
 AnimGraph {nodes [
   n1 = Get CurrentLOD()
   n2 = RigidBody(
    ComponentPose=Local To Component(LocalPose=Input Pose())
    Alpha=SelectFloat(A=0.2, B=1, bPickA="integer >= integer"!(A:int=@n1, B:int=1) -> (ReturnValue:bool) {is=title K2Node_CallFunction})
   ) {
    bClampLinearTranslationLimitToRefPose=true
    bOverrideWorldGravity=true
    overridePhysicsAsset=&/M_MED_Palespooky_Pack_Physics
    simulationSpace=BaseBoneSpace
    alphaBoolBlend {blendOption=Linear}
    baseBoneRef {boneName=root}
    componentAppliedLinearAccClamp=<3000 3000 3000>
    componentLinearAccScale=<0.2 0.2 0.2>
    componentLinearVelScale=<0.2 0.2 0.2>
    overrideWorldGravity {z=-980}
   }
   Output Pose(Result=Component To Local(ComponentPose=@n2)) {layerGroup=None}
  ]}
 EventGraph {nodes [TryGetPawnOwner()]}
]
variables [
 backpack_rotate {type=Rotator}
 backpack_translation {type=Vector}
]
